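# Master unlit additive material for Niagara sprite and mesh particles in impact effects, built around a muzzle-flash shape texture.
# A selectable shape channel is modulated by panned noise with optional UV transpose and distortion, then tinted by particle color.
# That color passes through optional HDR remapping and shared emissive adjustments.
# A separate alpha chain combines shape and particle alpha with switchable inverted Fresnel (bypassed at Low quality), plus scalar-blended depth and near-camera fades.
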Material /Game/Effects/Niagara/Emitters/Impacts/Materials/M_Core_Master
graph {
 properties {
  bUsedWithNiagaraMeshParticles=true
  bUsedWithNiagaraSprites=true
  blendMode=BLEND_Additive
  pixelDepthOffsetMode=PDOM_Legacy
  shadingModel=MSM_Unlit
  usedShadingModels=Unlit
 }
 nodes [
  n1 = MaterialFunctionCall(Invert Fresnel=StaticBool() {value=true}) {materialFunction=&Fresnel_Function}
  n2 = ScalarParameter() {defaultValue=4, group=Fresnel, parameterName=Exponent}
  n3 = Saturate(None=Multiply(A=Power(Base=@n1.Result, Exp=@n2) {constExponent=4}, B=ScalarParameter() {group=Fresnel, parameterName=Param_1, sortPriority=1}))
  n4 = Constant() {r=1}
  n5 = Multiply(
   A=Multiply(
    A=TextureSampleParameter2D() {group=MuzzleFlash, parameterName=MuzleFlash Shape, texture=/Game/Weapons/FORT_Pistols/Effects/Muzzle/T_Muzzle_Flash_02}.A
    B=ParticleColor().A
   )
   B=StaticSwitch(True=QualitySwitch(Default=@n3, Low=@n4, High=@n3, Medium=@n3), False=@n4)
  )
  n6 = ScalarParameter() {defaultValue=100, parameterName=FadeDistance}
  n7 = LinearInterpolate(A=@n5, B=Multiply(A=@n5, B=DepthFade(FadeDistance=@n6)), Alpha=ScalarParameter() {parameterName=UseDepthFade})
  n8 = Add(
   A=MaterialFunctionCall(Distance=ScalarParameter() {defaultValue=150, group=Camera Fade, parameterName=NearFadeDistance}) {materialFunction=&Near_Camera_Fade}.Result
   B=ScalarParameter() {parameterName=NearFadeAdjustment}
  )
  LinearInterpolate(A=@n7, B=Multiply(A=@n7, B=Saturate(None=@n8)), Alpha=ScalarParameter() {group=Camera Fade, parameterName=Use Camera Fade})
  n9 = ChannelMaskParameter(None=TextureSampleParameter2D() {group=MuzzleFlash, parameterName=MuzleFlash Shape, texture=&T_Muzzle_Flash_02}.RGB) {group=MuzzleFlash, maskChannel=Blue, parameterName=Channel, sortPriority=1, defaultValue=<rgba 0 0 1 0>}
  n10 = TextureCoordinate()
  n11 = AppendVector(
   A=ScalarParameter() {defaultValue=1, group=Noise, parameterName=U Distort Freq, sortPriority=5}
   B=ScalarParameter() {defaultValue=1, group=Noise, parameterName=V Distort Freq, sortPriority=6}
  )
  n12 = ComponentMask(None=TextureSampleParameter2D(UVs=Multiply(A=@n10, B=@n11)) {group=Noise, parameterName=Texture, sortPriority=4, texture=/Game/Weapons/FORT_Pistols/Effects/Muzzle/T_Cloud_d}.RGB) {g=true, r=true}
  n13 = TextureCoordinate()
  n14 = ComponentMask(None=@n13) {g=true}
  n15 = ComponentMask(None=@n13) {r=true}
  n16 = MaterialFunctionCall!(
   Texture Coordinates=StaticSwitchParameter(True=AppendVector(A=@n14, B=@n15), False=@n13) {group=Noise, parameterName=Transpose UVs}
   U Frequency=ScalarParameter() {defaultValue=1, group=Noise, parameterName=U Frequency, sortPriority=1}
   V Frequency=ScalarParameter() {defaultValue=1, group=Noise, parameterName=V Frequency, sortPriority=2}
   Offset
   OffsetMultiplier
  ) {materialFunction=/Game/Effects/Niagara/MasterMaterials/Functions/MF_Niagara_Panner}
  n17 = StaticSwitchParameter(
   True=Multiply(A=Add(A=@n12, B=@n16.Result), B=ScalarParameter() {group=Noise, parameterName=Distortion Amplitude, sortPriority=7})
   False=@n16.Result
  ) {group=Noise, parameterName=Enable UV Distortion, sortPriority=3}
  n18 = Saturate(None=Power(
    Base=TextureSampleParameter2D(UVs=@n17) {group=Noise, parameterName=Noise, texture=/Game/Global/Textures/Default/Tileable_Patterns/T_Tornado_Tile}.RGB
    Exp=ScalarParameter() {defaultValue=1, group=Noise, parameterName=Noise Exp}
   ) {constExponent=1.5})
  n19 = Multiply(
   A=LinearInterpolate(A=@n9.R, B=Multiply(A=@n9.R, B=@n18), Alpha=ScalarParameter() {group=Noise, parameterName=Noise Amount})
   B=ParticleColor().RGB
  )
  n20 = Clamp(None=@n19)
  MaterialFunctionCall(Emissive=StaticSwitchParameter(
    True=MaterialFunctionCall(LDR Input=@n20, HDR Tint=@n19) {materialFunction=&DeriveHDRfromLDR}.RemappedRange
    False=@n19
   ) {parameterName=Use HDR}) {materialFunction=&MF_FX_EmissiveAdjustments}
 ]
}
